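# Translucent UI master material for a soft flare glow, filed under the Onboarding UI.
# An exponential radial gradient with adjustable radius and density forms the round glow, while the product of sine-warped U and V forms a squarer one.
# SquaredGlow blends between the two shapes in two differently powered masks.
# One mask is a standalone scalar falloff; the other interpolates between the RGB 1 and RGB 2 color parameters.
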Material /Game/UI/Foundation/Onboarding/Materials/M_FX_Flare_Master
graph {
 properties {
  bUsedWithStaticMesh=false
  blendMode=BLEND_Translucent
  materialDomain=MD_UI
  pixelDepthOffsetMode=PDOM_Legacy
  usedShadingModels=Default Lit
 }
 nodes [
  n1 = MaterialFunctionCall(
   CenterPosition=Constant2Vector() {g=0.5, r=0.5}
   Radius=ScalarParameter() {defaultValue=0.5, parameterName=Radius, sliderMax=1}
   Density=ScalarParameter() {defaultValue=1.25, parameterName=Density, sliderMax=2}
  ) {materialFunction=&RadialGradientExponential}
  n2 = Sine(None=TextureCoordinate()) {period=2}
  n3 = DotProduct(A=ComponentMask(None=@n2) {r=true}, B=ComponentMask(None=@n2) {g=true})
  n4 = ScalarParameter() {parameterName=SquaredGlow, sliderMax=1}
  LinearInterpolate(A=Power(Base=@n1.RadialGradientExponential) {constExponent=2.5}, B=Power(Base=@n3) {constExponent=1.25}, Alpha=@n4)
  LinearInterpolate(
   A=VectorParameter() {parameterName=RGB 1, defaultValue=<rgba 0 1 1 1>}.RGB
   B=VectorParameter() {parameterName=RGB 2, defaultValue=<rgba 1 1 1 1>}.RGB
   Alpha=LinearInterpolate(A=Power(Base=@n1.RadialGradientExponential), B=@n3, Alpha=@n4)
  )
 ]
}
