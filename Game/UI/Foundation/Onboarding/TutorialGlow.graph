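# UI material for an animated translucent glow, apparently used to highlight elements such as buttons during onboarding tutorials.
# Two samples of one caustics texture rotate in opposite directions at different speeds and combine through a Difference blend to form a shifting pattern.
# That pattern interpolates between two parameterized pale blue colors.
# A swappable Template texture, defaulting to a button shape and scaled by an Opacity parameter, provides the mask.
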Material /Game/UI/Foundation/Onboarding/TutorialGlow
graph {
 properties {
  bUsedWithStaticMesh=false
  blendMode=BLEND_Translucent
  materialDomain=MD_UI
  pixelDepthOffsetMode=PDOM_Legacy
  shadingModel=MSM_Unlit
  usedShadingModels=Unlit
 }
 nodes [
  n1 = Rotator() {speed=0.5}
  n2 = Rotator() {speed=-0.75}
  LinearInterpolate(
   A=VectorParameter() {parameterName=Color 1, defaultValue=<rgba 0.346 0.794 0.865 1>}.RGB
   B=VectorParameter() {parameterName=Color 2, defaultValue=<rgba 0.86 0.9846 1 1>}.RGB
   Alpha=MaterialFunctionCall!(
    Blend=TextureSample(UVs=@n1) {texture=&/TutorialGlow-Caustics-128}.RGB
    Base=TextureSample(UVs=@n2) {texture=&TutorialGlow-Caustics-128}.RGB
   ) {materialFunction=/Engine/Functions/Engine_MaterialFunctions03/Blends/Blend_Difference}.Result
  )
  n3 = TextureSampleParameter2D() {parameterName=Template, samplerType=SAMPLERTYPE_LinearColor, texture=&/TutorialGlow-Template-Button-128}
  Multiply(A=@n3.RGB, B=ScalarParameter() {defaultValue=1, parameterName=Opacity})
 ]
}
